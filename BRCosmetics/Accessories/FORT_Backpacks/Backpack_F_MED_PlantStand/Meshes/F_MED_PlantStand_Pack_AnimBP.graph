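# Animation Blueprint for the PlantStand backpack, spinning dyn_Spin while counter-rotating its earth, water, fire and air bones.
# The event graph accumulates both spin angles at a fixed rate when skydiving, otherwise easing speed with movement, faster when tactical sprinting.
# Owner spine bend drives a root offset and roll, applied only during skydiving.
# The four-node Anim Dynamics chain is not connected to the output pose.
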AnimBlueprint /BRCosmetics/Accessories/FORT_Backpacks/Backpack_F_MED_PlantStand/Meshes/F_MED_PlantStand_Pack_AnimBP
target_skeleton=&/F_MED_PlantStand_Pack_Skeleton
graphs [
 AnimGraph {nodes [
   n1 = Anim Dynamics - Bone: dyn_earth_mstr(Alpha=1, ExternalForce="0,0,0") {
    angularDampingOverride=0.999
    angularSpringConstant=150
    bAngularSpring=true
    bGravityOverrideInSimSpace=true
    bLinearSpring=true
    bOverrideAngularDamping=true
    bOverrideLinearDamping=true
    bUseGravityOverride=true
    gravityScale=0
    linearDampingOverride=0.995
    linearSpringConstant=50
    simulationSpace=RootRelative
    alphaBoolBlend {blendOption=Linear}
    alphaScaleBiasClamp {interpSpeedDecreasing=3, interpSpeedIncreasing=5}
    boundBone {boneName=dyn_earth_mstr}
    componentAppliedLinearAccClamp=<6000 6000 6000>
    componentLinearAccScale=<0.05 0.05 0.05>
    componentLinearVelScale=<0.8 0.8 0.6>
    gravityOverride=<0.25 0.25 0.25>
    physicsBodyDefinitions [{
      sphereCollisionRadius=0
      boundBone {boneName=dyn_earth_mstr}
      boxExtents=<50 50 50>
      constraintSetup {angularLimitsMax=<0 0 0>, angularLimitsMin=<0 0 0>, angularTarget=<0 0 0>, linearAxesMax=<1 0.5 1>, linearAxesMin=<-1 -0.5 -1>}
      localJointOffset=<0 -3 0>
     }]
    relativeSpaceBone {boneName=root}
   }
   n2 = Anim Dynamics - Bone: dyn_water_mstr(ComponentPose=@n1, Alpha=0.5, ExternalForce="0,0,0") {
    angularDampingOverride=0.999
    angularSpringConstant=150
    bAngularSpring=true
    bGravityOverrideInSimSpace=true
    bLinearSpring=true
    bOverrideAngularDamping=true
    bOverrideLinearDamping=true
    bUseGravityOverride=true
    gravityScale=0
    linearDampingOverride=0.995
    linearSpringConstant=50
    simulationSpace=RootRelative
    alphaBoolBlend {blendOption=Linear}
    alphaScaleBiasClamp {interpSpeedDecreasing=3, interpSpeedIncreasing=5}
    boundBone {boneName=dyn_water_mstr}
    componentAppliedLinearAccClamp=<6000 6000 6000>
    componentLinearAccScale=<0.05 0.05 0.05>
    componentLinearVelScale=<0.8 0.8 0.6>
    gravityOverride=<0.25 0.25 0.25>
    physicsBodyDefinitions [{
      sphereCollisionRadius=0
      boundBone {boneName=dyn_water_mstr}
      boxExtents=<50 50 50>
      constraintSetup {angularLimitsMax=<0 0 0>, angularLimitsMin=<0 0 0>, angularTarget=<0 0 0>, linearAxesMax=<1 0.5 1>, linearAxesMin=<-1 -0.5 -1>}
      localJointOffset=<0 -3 0>
     }]
    relativeSpaceBone {boneName=root}
   }
   n3 = Anim Dynamics - Bone: dyn_fire_mstr(ComponentPose=@n2, Alpha=0.75, ExternalForce="0,0,0") {
    angularDampingOverride=0.999
    angularSpringConstant=150
    bAngularSpring=true
    bGravityOverrideInSimSpace=true
    bLinearSpring=true
    bOverrideAngularDamping=true
    bOverrideLinearDamping=true
    bUseGravityOverride=true
    gravityScale=0
    linearDampingOverride=0.995
    linearSpringConstant=50
    simulationSpace=RootRelative
    alphaBoolBlend {blendOption=Linear}
    alphaScaleBiasClamp {interpSpeedDecreasing=3, interpSpeedIncreasing=5}
    boundBone {boneName=dyn_fire_mstr}
    componentAppliedLinearAccClamp=<6000 6000 6000>
    componentLinearAccScale=<0.05 0.05 0.05>
    componentLinearVelScale=<0.8 0.8 0.6>
    gravityOverride=<0.25 0.25 0.25>
    physicsBodyDefinitions [{
      sphereCollisionRadius=0
      boundBone {boneName=dyn_fire_mstr}
      boxExtents=<50 50 50>
      constraintSetup {angularLimitsMax=<0 0 0>, angularLimitsMin=<0 0 0>, angularTarget=<0 0 0>, linearAxesMax=<1 0 1>, linearAxesMin=<-1 -0.5 -1>}
      localJointOffset=<0 -3 0>
     }]
    relativeSpaceBone {boneName=root}
   }
   Anim Dynamics - Bone: dyn_air_mstr(ComponentPose=@n3, Alpha=0.8, ExternalForce="0,0,0") {
    angularDampingOverride=0.999
    angularSpringConstant=150
    bAngularSpring=true
    bGravityOverrideInSimSpace=true
    bLinearSpring=true
    bOverrideAngularDamping=true
    bOverrideLinearDamping=true
    bUseGravityOverride=true
    gravityScale=0
    linearDampingOverride=0.995
    linearSpringConstant=50
    simulationSpace=RootRelative
    alphaBoolBlend {blendOption=Linear}
    alphaScaleBiasClamp {interpSpeedDecreasing=3, interpSpeedIncreasing=5}
    boundBone {boneName=dyn_air_mstr}
    componentAppliedLinearAccClamp=<6000 6000 6000>
    componentLinearAccScale=<0.05 0.05 0.05>
    componentLinearVelScale=<0.8 0.8 0.6>
    gravityOverride=<0.25 0.25 0.25>
    physicsBodyDefinitions [{
      sphereCollisionRadius=0
      boundBone {boneName=dyn_air_mstr}
      boxExtents=<50 50 50>
      constraintSetup {angularLimitsMax=<0 0 0>, angularLimitsMin=<0 0 0>, angularTarget=<0 0 0>, linearAxesMax=<1 0 1>, linearAxesMin=<-1 -0.5 -1>}
      localJointOffset=<0 -3 0>
     }]
    relativeSpaceBone {boneName=root}
   }
   n4 = "Transform (Modify) Bone - Bone: dyn_Spin"(
    ComponentPose=Local To Component(LocalPose=Input Pose())
    Alpha=1
    Translation="0,0,0"
    Rotation=Get AllRotate() -> (AllRotate:Rotator)
    Scale="1,1,1"
   ) {
    rotationMode=BMM_Additive
    rotationSpace=BCS_BoneSpace
    translationMode=BMM_Additive
    translationSpace=BCS_BoneSpace
    alphaBoolBlend {blendOption=Linear}
    alphaScaleBiasClamp {bInterpResult=true, interpSpeedDecreasing=2, interpSpeedIncreasing=200, scale=300}
    boneToModify {boneName=dyn_Spin}
   }
   n5 = Get CounterRotate() -> (CounterRotate:Rotator)
   n6 = "Transform (Modify) Bone - Bone: dyn_earth_mstr"(ComponentPose=@n4, Alpha=1, Translation="0,0,0", Rotation=@n5, Scale="1,1,1") {
    rotationMode=BMM_Additive
    rotationSpace=BCS_BoneSpace
    translationMode=BMM_Additive
    alphaBoolBlend {blendOption=Linear}
    alphaScaleBiasClamp {bInterpResult=true, interpSpeedDecreasing=4, interpSpeedIncreasing=6}
    boneToModify {boneName=dyn_earth_mstr}
   }
   n7 = "Transform (Modify) Bone - Bone: dyn_water_mstr"(ComponentPose=@n6, Alpha=1, Translation="0,0,0", Rotation=@n5, Scale="1,1,1") {
    rotationMode=BMM_Additive
    rotationSpace=BCS_BoneSpace
    translationMode=BMM_Additive
    alphaBoolBlend {blendOption=Linear}
    alphaScaleBiasClamp {bInterpResult=true, interpSpeedDecreasing=4, interpSpeedIncreasing=6}
    boneToModify {boneName=dyn_water_mstr}
   }
   n8 = "Transform (Modify) Bone - Bone: dyn_fire_mstr"(ComponentPose=@n7, Alpha=1, Translation="0,0,0", Rotation=@n5, Scale="1,1,1") {
    rotationMode=BMM_Additive
    rotationSpace=BCS_BoneSpace
    translationMode=BMM_Additive
    alphaBoolBlend {blendOption=Linear}
    alphaScaleBiasClamp {bInterpResult=true, interpSpeedDecreasing=4, interpSpeedIncreasing=6}
    boneToModify {boneName=dyn_fire_mstr}
   }
   n9 = "Transform (Modify) Bone - Bone: dyn_air_mstr"(ComponentPose=@n8, Alpha=1, Translation="0,0,0", Rotation=@n5, Scale="1,1,1") {
    rotationMode=BMM_Additive
    rotationSpace=BCS_BoneSpace
    translationMode=BMM_Additive
    alphaBoolBlend {blendOption=Linear}
    alphaScaleBiasClamp {bInterpResult=true, interpSpeedDecreasing=4, interpSpeedIncreasing=6}
    boneToModify {boneName=dyn_air_mstr}
   }
   n10 = "Transform (Modify) Bone - Bone: root"(
    ComponentPose=@n9
    Alpha=Get PackAlpha() -> (PackAlpha:double)
    Translation=Get BackpackTranslate() -> (BackpackTranslate:Vector)
    Rotation=Get BackpackRotate() -> (BackpackRotate:Rotator)
   ) {
    rotationMode=BMM_Additive
    rotationSpace=BCS_BoneSpace
    translationMode=BMM_Additive
    translationSpace=BCS_BoneSpace
    alphaBoolBlend {blendOption=Linear}
    boneToModify {boneName=root}
   }
   Output Pose(Result=Component To Local(ComponentPose=@n10)) {layerGroup=None}
  ]}
 EventGraph {nodes [
   n1 = Event BlueprintUpdateAnimation()
   > n2 = Sequence()
   > n3 = Branch(Condition=Get bIsSkyDiving())
   > n4 = Sequence()
   Sequence(execute=@n4.then_1)
   n5 = float * float(A=-1, B=Lerp(A=60, B=60, Alpha=0))
   > n6 = Branch(Condition="float >= float"(A=Get CounterSpin() -> (CounterSpin:double), B=@n5))
   n7 = float + float(A=Get CounterSpin(), B=float * float(A=@n5, B=@n1.DeltaTimeX) {is=title K2Node_PromotableOperator}) {is=title K2Node_CommutativeAssociativeBinaryOperator}
   > n8 = Set CounterSpin(CounterSpin:double=@n7) -> (Output_Get:double)
   Set CounterRotate(execute=[@n8.then, @n9.then], CounterRotate:Rotator=MakeRotator(Pitch=@n7)) -> (Output_Get:Rotator)
   n9 = Set CounterSpin(execute=@n6.else, CounterSpin=@n7)
   n10 = Sequence(execute=@n4.then_0)
   n11 = float * float(A=1, B=Lerp(A=60, B=60, Alpha=0)) {is=title K2Node_CommutativeAssociativeBinaryOperator}
   n12 = Branch(execute=@n10.then_1, Condition="float >= float"(A=Get Spin() -> (Spin:double), B=@n11))
   n13 = float + float(A=Get Spin(), B=float * float(A=@n11, B=@n1.DeltaTimeX) {is=title K2Node_PromotableOperator})
   > n14 = Set Spin(Spin:double=@n13) -> (Output_Get:double)
   Set AllRotate(execute=[@n14.then, @n15.then], AllRotate:Rotator=MakeRotator(Pitch=@n13)) -> (Output_Get:Rotator)
   n15 = Set Spin(execute=@n12.else, Spin=@n13)
   Reroute Node(InputPin=@n10.then_0)
   n16 = Branch(execute=@n3.else, Condition=Get bIsTacticalSprinting())
   > n17 = Sequence()
   n18 = Sequence(execute=@n17.then_1)
   n19 = Get bIsMoving2D()
   n20 = Set Running(
    execute=@n18.then_1
    Running:double=FInterpTo(Current=Get Running() -> (Running:double), Target="To Float (Boolean)"(InBool=@n19), DeltaTime=@n1.DeltaTimeX, InterpSpeed=1.5)
   ) -> (Output_Get:double)
   n21 = float * float(A=-1, B=Lerp(A=20, B=125, Alpha=@n20)) {is=title K2Node_CommutativeAssociativeBinaryOperator}
   n22 = Branch(execute=@n18.then_0, Condition="float >= float"(A=Get CounterSpin(), B=@n21))
   n23 = float + float(A=Get CounterSpin(), B=float * float(A=@n21, B=@n1.DeltaTimeX) {is=title K2Node_PromotableOperator})
   > n24 = Set CounterSpin(CounterSpin=@n23)
   Set CounterRotate(execute=[@n24.then, @n25.then], CounterRotate=MakeRotator(Pitch=@n23))
   n25 = Set CounterSpin(execute=@n22.else, CounterSpin=@n23)
   n26 = Sequence(execute=@n17.then_0)
   n27 = Get bIsMoving2D()
   n28 = Set Running(
    execute=@n26.then_1
    Running=FInterpTo(Current=Get Running(), Target="To Float (Boolean)"(InBool=@n27), DeltaTime=@n1.DeltaTimeX, InterpSpeed=1.5)
   )
   n29 = float * float(A=1, B=Lerp(A=20, B=125, Alpha=@n28)) {is=title K2Node_CommutativeAssociativeBinaryOperator}
   n30 = Branch(execute=@n26.then_0, Condition="float >= float"(A=Get Spin(), B=@n29))
   n31 = float + float(A=Get Spin(), B=float * float(A=@n29, B=@n1.DeltaTimeX) {is=title K2Node_PromotableOperator})
   > n32 = Set Spin(Spin=@n31)
   Set AllRotate(execute=[@n32.then, @n33.then], AllRotate=MakeRotator(Pitch=@n31))
   n33 = Set Spin(execute=@n30.else, Spin=@n31)
   n34 = Sequence(execute=@n16.else)
   n35 = Sequence(execute=@n34.then_1)
   n36 = Get bIsMoving2D()
   n37 = Set Running(
    execute=@n35.then_1
    Running=FInterpTo(Current=Get Running(), Target="To Float (Boolean)"(InBool=@n36), DeltaTime=@n1.DeltaTimeX, InterpSpeed=1.5)
   )
   n38 = float * float(A=-1, B=Lerp(A=20, B=100, Alpha=@n37)) {is=title K2Node_CommutativeAssociativeBinaryOperator}
   n39 = Branch(execute=@n35.then_0, Condition="float >= float"(A=Get CounterSpin(), B=@n38))
   n40 = float + float(A=Get CounterSpin(), B=float * float(A=@n38, B=@n1.DeltaTimeX) {is=title K2Node_PromotableOperator})
   > n41 = Set CounterSpin(CounterSpin=@n40)
   Set CounterRotate(execute=[@n42.then, @n41.then], CounterRotate=MakeRotator(Pitch=@n40))
   n42 = Set CounterSpin(execute=@n39.else, CounterSpin=@n40)
   n43 = Sequence(execute=@n34.then_0)
   n44 = Get bIsMoving2D()
   n45 = Set Running(
    execute=@n43.then_1
    Running=FInterpTo(Current=Get Running(), Target="To Float (Boolean)"(InBool=@n44), DeltaTime=@n1.DeltaTimeX, InterpSpeed=1.5)
   )
   n46 = float * float(A=1, B=Lerp(A=20, B=100, Alpha=@n45)) {is=title K2Node_CommutativeAssociativeBinaryOperator}
   n47 = Branch(execute=@n43.then_0, Condition="float >= float"(A=Get Spin(), B=@n46))
   n48 = float + float(A=Get Spin(), B=float * float(A=@n46, B=@n1.DeltaTimeX) {is=title K2Node_PromotableOperator})
   > n49 = Set Spin(Spin=@n48)
   Set AllRotate(execute=[@n49.then, @n50.then], AllRotate=MakeRotator(Pitch=@n48))
   n50 = Set Spin(execute=@n47.else, Spin=@n48)
   n51 = GetOwnerSkeletalMeshComponent()
   n52 = BreakRotator(InRot=BreakTransform(InTransform=GetDeltaTransformFromRefPose(self=@n51, BoneName=spine_05, BaseName=spine_01)).Rotation)
   Set BackpackRotate(execute=@n2.then_1, BackpackRotate:Rotator=MakeRotator(Roll=MapRangeClamped(Value=@n52.Yaw, InRangeA=0, InRangeB=-30, OutRangeB=35))) -> (Output_Get:Rotator)
   > Set BackpackTranslate(BackpackTranslate:Vector=MakeVector(Y=MapRangeClamped(Value=@n52.Yaw, InRangeA=0, InRangeB=-30, OutRangeB=-10))) -> (Output_Get:Vector)
   > n53 = Branch(Condition=Get bIsSkyDiving())
   > Set PackAlpha(PackAlpha:double=1) -> (Output_Get:double)
   Set PackAlpha(execute=@n53.else, PackAlpha=0)
   Event BlueprintInitializeAnimation()
   > Set Skel_Mesh(Skel_Mesh=GetOwnerSkeletalMeshComponent())
   > n54 = Cast To FortPlayerPawn(Object=TryGetPawnOwner())
   > Set Pawn(Pawn:Fort Player Pawn Object Reference=@n54) -> (Output_Get:Fort Player Pawn Object Reference)
  ]}
 spinFNC_deprecated {nodes [
   spinFNC_deprecated!() -> (then, speed:double=~, initVal:double=~) {is=title K2Node_FunctionEntry}
   > Return Node!(execute, output:double=0)
  ]}
]
variables [
 AllRotate {type=Rotator}
 BackpackRotate {type=Rotator}
 BackpackTranslate {type=Vector}
 CounterRotate {type=Rotator}
 CounterSpin {type=number}
 NewVar {type=number}
 PackAlpha {type=number}
 Pawn {type=FortPlayerPawn}
 Running {type=number}
 Skel_Mesh {type=SkeletalMeshComponent}
 Spin {type=number}
]
